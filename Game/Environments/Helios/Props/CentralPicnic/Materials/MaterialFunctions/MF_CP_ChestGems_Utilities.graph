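# Shared utility function for Central Picnic chest gem materials, outputting values for a camera-reactive orb inside glass.
# It projects an offset object-space center, optionally shifted when viewed from below, into screen UVs that add a WarpMask input; the screen-ratio term divides view height by itself.
# The center also drives distance masks, view-angle pan and bottom-up alpha, alongside an FOV/scale multiplier.
# Glass outputs are color, Fresnel thickness mask and roughness with optional noise breakup.
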MaterialFunction /Game/Environments/Helios/Props/CentralPicnic/Materials/MaterialFunctions/MF_CP_ChestGems_Utilities
graph {nodes [
  n1 = FunctionInput(Preview=Constant()) {bUsePreviewValueAsDefault=true, inputName=WarpMask, inputType=FunctionInput_Scalar}
  ComponentMask(None=@n1) {b=true}
  n2 = Add(
   A=ObjectPositionWS()
   B=Transform(None=VectorParameter() {group=Center Orb | General, parameterName=Orb Center Relative Offset, defaultValue=<rgba 0 0 0 1>}.RGB) {transformSourceType=TRANSFORMSOURCE_Local}
  ) {constB=0}
  n3 = VectorParameter() {group=Center Orb | General, parameterName=Bottom Up Offset, defaultValue=<rgba 0 0 0 1>}
  n4 = ObjectPositionWS()
  n5 = Transform(None=VectorParameter() {group=Center Orb | General, parameterName=Orb Center Relative Offset, defaultValue=<rgba 0 0 0 1>}.RGB) {transformSourceType=TRANSFORMSOURCE_Local}
  n6 = Normalize(VectorInput=Subtract(A=Add(A=@n4, B=@n5) {constB=0}, B=CameraPositionWS()))
  n7 = ComponentMask(None=@n6) {b=true}
  n8 = If(A=@n7, "A > B"=@n7, "A < B"=Constant())
  n9 = StaticSwitchParameter(True=Transform(None=LinearInterpolate(B=@n3.RGB, Alpha=@n8) {constB=0}) {transformSourceType=TRANSFORMSOURCE_Local}, False=Constant()) {
   defaultValue=true
   desc=Offset the center point when looking from the bottom
   group=Center Orb | General
   parameterName=Bottom Up Offset Center
  }
  n10 = MaterialFunctionCall(Float2=MaterialFunctionCall(World Position=Add(A=@n2, B=@n9)) {materialFunction=&TransformToClipSpace}.Clip Space XY) {materialFunction=&BreakOutFloat2Components}
  n11 = Divide(A=Add(A=@n10.R))
  n12 = Constant() {r=-0.5}
  n13 = MaterialFunctionCall(Float2=ViewSize()) {materialFunction=&BreakOutFloat2Components}
  n14 = Divide(A=@n13.G, B=@n13.G)
  n15 = Divide(A=OneMinus(None=@n14))
  n16 = Add(A=Multiply(A=@n10.G, B=Add(A=@n12, B=@n15) {desc=Account for Screen Ratio Offset}) {constB=-0.25}) {constB=0.5}
  FunctionOutput(None=Add(A=@n1, B=AppendVector(A=@n11, B=@n16))) {outputName=Orb Center UV Space, sortPriority=2}
  MaterialExpressionNamedRerouteDeclaration_9 = NamedRerouteDeclaration(None=@n2) {name=Orb Center World Position, nodeColor=<rgba 1 0 0.0938 1>}
  n17 = Custom!(In=@MaterialExpressionNamedRerouteDeclaration_9) {
   code="mul(float4(In.xyz,1),LWCToFloat(ResolvedView.WorldToClip))"
   desc="\r\n*** Original source before expression upgrade ***\r\nmul(float4(In.xyz,1),View.WorldToClip)"
   outputType=CMOT_Float4
   inputs [{
     inputName=In
     input {expression=&self:MaterialExpressionNamedRerouteDeclaration_9}
    }]
  }
  n18 = ComponentMask(None=@n17) {a=true}
  FunctionOutput(None=NamedRerouteDeclaration(None=Divide(B=@n18) {constA=50, constB=1000}) {name=Center Orb UV Distance, nodeColor=<rgba 0 1 0.0391 1>}) {outputName=Center Orb Distance, sortPriority=3}
  n19 = VectorParameter() {group=_Colors, parameterName=Glass Color, sortPriority=12, defaultValue=<rgba 0.3906 0.1417 0.0367 1>}
  FunctionOutput(None=@n19.RGB) {outputName=Glass Color, sortPriority=9}
  n20 = MaterialFunctionCall(Float2=ScreenPosition().ViewportUV) {materialFunction=&BreakOutFloat2Components}
  n21 = MaterialFunctionCall(X=@n20.R, Y=Add(A=Multiply(A=@n20.G, B=@n14), B=@n15)) {materialFunction=&MakeFloat2}
  FunctionOutput(None=Add(A=@n1, B=@n21.Result)) {outputName=Screen Scaled UVs, sortPriority=2}
  n22 = Distance(A=@MaterialExpressionNamedRerouteDeclaration_9, B=CameraPositionWS())
  n23 = ScalarParameter() {parameterName=General Distance Mask Offset}
  n24 = Saturate(None=Divide(A=Add(A=@n22, B=@n23), B=ScalarParameter() {defaultValue=1000, parameterName=General Distance Mask Fade Distance}))
  FunctionOutput(None=@n24) {outputName=Distance Mask, sortPriority=4}
  n25 = MaterialFunctionCall(
   Input=@n8
   Input Low=ScalarParameter() {defaultValue=0.25, group=Center Orb | General, parameterName=Bottom Up Start Alpha}
   Input High=Constant() {r=1}
  ) {materialFunction=&RemapValue_0-1}
  FunctionOutput(None=NamedRerouteDeclaration(None=Power(Base=@n25.Result) {constExponent=3}) {name=Bottom Up Alpha, nodeColor=<rgba 0.9922 0 1 1>}) {outputName=Bottom Up Alpha, sortPriority=6}
  n26 = Multiply(
   A=Divide(B=MaterialFunctionCall() {materialFunction=/Engine/Functions/Engine_MaterialFunctions02/Utility/FOV}.FOV angle in degrees) {constA=90, constB=90}
   B=MaterialFunctionCall() {materialFunction=&ObjectScale}.Scale X
  )
  n27 = ScalarParameter() {defaultValue=1, group=_General, parameterName=GlobalScaleMult, sortPriority=1}
  FunctionOutput(None=NamedRerouteDeclaration(None=Multiply(A=@n26, B=@n27)) {name=FOV_Object Scale Multi, nodeColor=<rgba 1 0.2812 0 1>}) {outputName=FOV - Scale Multiplier, sortPriority=7}
  n28 = MaterialFunctionCall(Vector or UVs=@n6) {materialFunction=&VectorToRadialValue}
  n29 = MaterialFunctionCall(
   X=Multiply(
    A=FunctionOutput(None=@n28.Vector Converted to Angle) {outputName=XY Rotation Alpha, sortPriority=5}
    B=ScalarParameter() {defaultValue=3.6, group=_General, parameterName=Camera Rotation Pan Multi X, sortPriority=40}
   ) {constB=-1}
   Y=Multiply(A=@n7, B=ScalarParameter() {defaultValue=-1, group=_General, parameterName=Camera Rotation Pan Multi Y, sortPriority=40}) {constB=-1}
  ) {materialFunction=&MakeFloat2}
  FunctionOutput(None=NamedRerouteDeclaration(None=Multiply(A=@n29.Result) {constB=-1}) {name=XY Rotation Alpha Around Center, nodeColor=<rgba 0.1953 0 1 1>}) {outputName=XY Rotation Offset Amount, sortPriority=5}
  n30 = OneMinus(None=TextureSample() {samplerType=SAMPLERTYPE_Masks, texture=/Game/Environments/Asteria/Terrain/Textures/T_Asteria_Ice_Noise_M}.R)
  n31 = ScalarParameter() {defaultValue=0.05, group=Glass, parameterName=Roughness}
  FunctionOutput(None=StaticSwitchParameter(True=Add(A=@n30, B=@n31), False=@n31) {group=_General, parameterName=Roughness Breakup}) {outputName=Gem Roughness, sortPriority=8}
  n32 = Fresnel(ExponentIn=ScalarParameter() {defaultValue=5, group=Glass, parameterName=Glass Exp In, sortPriority=30}) {baseReflectFraction=0}
  n33 = Power(Base=@n32, Exp=ScalarParameter() {defaultValue=2, group=Glass, parameterName=Glass Softness})
  n34 = ScalarParameter() {defaultValue=10, group=Glass, parameterName=Glass Thickness}
  n35 = Saturate(None=Add(A=Multiply(A=@n33, B=@n34) {constB=2}, B=Saturate(None=@n32)))
  n36 = ScalarParameter() {defaultValue=0.6, group=Glass, parameterName=Glass Edge Opacity, sliderMax=1, sortPriority=12}
  FunctionOutput(None=NamedRerouteDeclaration(None=Multiply(A=@n35, B=@n36)) {name=GlassThicknessMask, nodeColor=<rgba 0 1 0.6719 1>}) {outputName=Glass Thickness Mask, sortPriority=10}
  FunctionOutput(None=@MaterialExpressionNamedRerouteDeclaration_9) {outputName=Orb Center World Pos, sortPriority=1}
  LinearInterpolate(A=@n31, B=Constant() {r=0.3})
  NamedRerouteDeclaration(None=@n21.Result) {name=ScreenScaledUVs, nodeColor=<rgba 0.6875 1 0 1>}
  NamedRerouteDeclaration(None=@n24) {name=Distance Mask, nodeColor=<rgba 1 0.4219 0 1>}
 ]}
