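# Material function that shades the glowing center orb of the CentralPicnic chest gems, blended onto input attributes by a mask.
# Screen-space UVs anchored to the orb center layer background constellations, noise-warped radial wisps, gradients and Fresnel highlights.
# Front constellations are flipbooks with sine-driven warp and fade plus noise pulses.
# It also outputs unmasked attributes and a constellation-wisp mask; switches cut wisp distortion and upper gradients on low-end.
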MaterialFunction /Game/Environments/Helios/Props/CentralPicnic/Materials/MaterialFunctions/MF_CP_ChestGems_CenterGem
graph {nodes [
  n1 = LinearInterpolate(
   B=LinearInterpolate(A=NamedRerouteUsage(), B=NamedRerouteUsage())
   Alpha=FunctionInput(Preview=Constant() {r=1}) {bUsePreviewValueAsDefault=true, inputName=Constellation Wisp Mask Opacity, inputType=FunctionInput_Scalar, sortPriority=5}
  )
  n2 = ScalarParameter() {defaultValue=1.7, group=Center Orb | Wisp, parameterName=Wisp Brightness Mask Radius, sortPriority=41}
  n3 = Multiply(A=NamedRerouteUsage(), B=NamedRerouteUsage())
  n4 = LinearInterpolate(Alpha=MaterialFunctionCall(
    UVs=NamedRerouteUsage()
    CenterPosition=NamedRerouteUsage()
    Radius=Multiply(A=@n2, B=@n3)
    Density=ScalarParameter() {defaultValue=1.5, group=Center Orb | Wisp, parameterName=Wisp Brightness Mask Density, sortPriority=41}
   ) {materialFunction=&RadialGradientExponential}.RadialGradientExponential) {constA=0.15}
  FunctionOutput(None=Max(A=Multiply(A=@n1, B=@n4), B=NamedRerouteUsage())) {outputName=ConstellationWispsMask, sortPriority=4}
  n5 = MakeMaterialAttributes!(
   BaseColor
   Metallic
   Specular
   Roughness
   Anisotropy
   EmissiveColor
   Opacity
   OpacityMask
   Normal
   Tangent
   WorldPositionOffset
   SubsurfaceColor
   ClearCoat
   ClearCoatRoughness
   AmbientOcclusion
   Refraction
   CustomizedUVs_0
   CustomizedUVs_1
   CustomizedUVs_2
   CustomizedUVs_3
   CustomizedUVs_4
   CustomizedUVs_5
   CustomizedUVs_6
   CustomizedUVs_7
   PixelDepthOffset
   ShadingModel
   Displacement
  )
  n6 = VectorParameter() {group=_Colors, parameterName=Background Color, sortPriority=13, defaultValue=<rgba 0.0781 0.0094 0.0049 1>}
  n7 = VectorParameter() {group=_Colors, parameterName=Center Orb Background Constellation Color, sortPriority=19, defaultValue=<rgba 1 0.2779 0 1>}
  n8 = NamedRerouteUsage()
  n9 = VectorParameter() {group=_Colors, parameterName=Center Orb Wisp Color, sortPriority=20, defaultValue=<rgba 0.8021 0.1849 0.0249 1>}
  n10 = VectorParameter() {group=_Colors, parameterName=Center Orb Wisp Color 2, sortPriority=20, defaultValue=<rgba 0.4583 0.0917 0.017 1>}
  n11 = LinearInterpolate(A=Multiply(A=@n9.RGB, B=NamedRerouteUsage()), B=@n10.RGB, Alpha=NamedRerouteUsage()) {constB=0}
  n12 = LinearInterpolate(
   A=LinearInterpolate(A=LinearInterpolate(A=@n6.RGB, B=@n7.RGB, Alpha=@n8), B=Multiply(A=@n11, B=@n4), Alpha=NamedRerouteUsage())
   B=NamedRerouteUsage()
   Alpha=NamedRerouteUsage()
  )
  n13 = NamedRerouteUsage()
  n14 = NamedRerouteUsage()
  n15 = VectorParameter() {group=_Colors, parameterName=Center Orb Highlight Color, sortPriority=22, defaultValue=<rgba 1 0.0868 0 1>}
  n16 = Multiply(
   A=Fresnel(ExponentIn=ScalarParameter() {defaultValue=5, group=Center Orb | Edge Gradient Highlight, parameterName=Edge Highlight Thickness}) {baseReflectFraction=0}
   B=ScalarParameter() {defaultValue=2, group=Center Orb | Edge Gradient Highlight, parameterName=Edge Highlight Intensity}
  )
  n17 = ScalarParameter() {defaultValue=0.5, group=Center Orb | General, parameterName=Center Orb Fresnel Exponent}
  n18 = Multiply(
   A=LinearInterpolate(A=LinearInterpolate(A=@n12, B=@n13, Alpha=@n14) {constA=-0.32}, B=@n15.RGB, Alpha=Saturate(None=@n16))
   B=Power(
    Base=MaterialFunctionCall(Exponent=@n17) {materialFunction=&MF_Fresnel_FX}.Fresnel
    Exp=ScalarParameter() {defaultValue=2, group=Center Orb | General, parameterName=Center Orb Fresnel Contrast}
   )
  )
  n19 = LinearInterpolate(A=@n18, B=Multiply(A=@n18) {constB=0.25}, Alpha=NamedRerouteUsage()) {constB=0}
  n20 = ScalarParameter() {defaultValue=4, group=Center Orb | General, parameterName=Center Orb Emissive Power}
  n21 = NamedRerouteUsage()
  n22 = NamedRerouteUsage()
  n23 = MakeMaterialAttributes(
   BaseColor=Multiply(
    A=LinearInterpolate(A=@n18, B=@n21, Alpha=@n22)
    B=ScalarParameter() {defaultValue=1, group=Center Orb | General, parameterName=Base Color Brightness}
   )
   Roughness=NamedRerouteUsage()
   EmissiveColor=Saturate(None=Multiply(A=@n19, B=@n20))
  )
  FunctionOutput(None=BlendMaterialAttributes(
    A=FunctionInput(Preview=@n5) {bUsePreviewValueAsDefault=true, inputName=Material, inputType=FunctionInput_MaterialAttributes}
    B=@n23
    Alpha=FunctionInput() {inputName=Mask, inputType=FunctionInput_Scalar, sortPriority=1}
   )) {outputName=Material}
  FunctionOutput(None=@n23) {outputName=Material Non Masked, sortPriority=1}
  MakeMaterialAttributes()
  n24 = Multiply(A=ComponentMask(None=NamedRerouteUsage()) {b=true}) {constB=10}
  n25 = MaterialFunctionCall!(WarpMask=@n24) {materialFunction=&/MF_CP_ChestGems_Utilities}
  n26 = ScalarParameter() {defaultValue=1.9, group=Center Orb | Gradients, parameterName=Gradient B Side Offset, sortPriority=44}
  n27 = Multiply(A=ScalarParameter() {defaultValue=1.6, group=Center Orb | Gradients, parameterName=Gradient B Height Offset, sortPriority=44}) {constB=-1}
  n28 = Multiply(
   A=@n25.Center Orb Distance
   B=Multiply(
    A=MaterialFunctionCall(X=@n26, Y=@n27) {materialFunction=&MakeFloat2}.Result
    B=@n25.FOV - Scale Multiplier
   )
  ) {constB=1.1}
  n29 = Multiply(
   A=@n25.Center Orb Distance
   B=Multiply(
    A=ScalarParameter() {defaultValue=2, group=Center Orb | Gradients, parameterName=Radial Gradient B Radius, sortPriority=42}
    B=@n25.FOV - Scale Multiplier
   )
  ) {constB=1.1}
  n30 = MaterialFunctionCall(
   UVs=@n25.Screen Scaled UVs
   CenterPosition=Add(A=@n28, B=@n25.Orb Center UV Space) {constB=0}
   Radius=@n29
   Density=ScalarParameter() {defaultValue=2.5, group=Center Orb | Gradients, parameterName=Radial Gradient B Density, sortPriority=43}
  ) {materialFunction=&RadialGradientExponential}
  n31 = ScalarParameter() {defaultValue=0.6, group=Center Orb | Gradients, parameterName=Gradient B Power, sortPriority=41}
  n32 = MaterialFunctionCall(Float2=@n28) {materialFunction=&BreakOutFloat2Components}
  n33 = Add(
   A=MaterialFunctionCall(X=Multiply(A=@n32.R) {constB=-1}, Y=@n32.G) {materialFunction=&MakeFloat2}.Result
   B=@n25.Orb Center UV Space
  ) {constB=0}
  n34 = ScalarParameter() {defaultValue=2.5, group=Center Orb | Gradients, parameterName=Radial Gradient B Density, sortPriority=43}
  n35 = Add(
   A=Multiply(
    A=MaterialFunctionCall(UVs=@n25.Screen Scaled UVs, CenterPosition=@n33, Radius=@n29, Density=@n34) {materialFunction=&RadialGradientExponential}.RadialGradientExponential
    B=ScalarParameter() {defaultValue=0.6, group=Center Orb | Gradients, parameterName=Gradient B Power, sortPriority=41}
   ) {constB=5}
   B=Saturate(None=Multiply(A=@n30.RadialGradientExponential, B=@n31) {constB=5})
  )
  n36 = FeatureLevelSwitch(Default=QualitySwitch(Default=Saturate(None=@n35))) {desc="Remove upper gradients on low ~26 inst"}
  n37 = Constant()
  n38 = MaterialFunctionCall(WarpMask=@n24) {materialFunction=&MF_CP_ChestGems_Utilities}
  n39 = Multiply(
   A=@n38.Center Orb Distance
   B=Multiply(
    A=ScalarParameter() {defaultValue=2.4, group=Center Orb | Gradients, parameterName=Gradient A Height Offset, sortPriority=23}
    B=@n38.FOV - Scale Multiplier
   )
  ) {constB=1.1}
  n40 = ScalarParameter() {defaultValue=3, group=Center Orb | Gradients, parameterName=Radial A Gradient Radius, sortPriority=21}
  n41 = MaterialFunctionCall(
   UVs=@n38.Screen Scaled UVs
   CenterPosition=Add(
    A=MaterialFunctionCall(X=@n37, Y=@n39) {materialFunction=&MakeFloat2}.Result
    B=@n38.Orb Center UV Space
   ) {constB=0}
   Radius=Multiply(A=@n38.Center Orb Distance, B=Multiply(A=@n40, B=@n38.FOV - Scale Multiplier)) {constB=1.1}
   Density=ScalarParameter() {defaultValue=1.4, group=Center Orb | Gradients, parameterName=Radial A Gradient Density, sortPriority=22}
  ) {materialFunction=&RadialGradientExponential}
  n42 = ScalarParameter() {defaultValue=1, group=Center Orb | Gradients, parameterName=Gradient A Power, sortPriority=20}
  n43 = MaterialFunctionCall(Exponent=ScalarParameter() {defaultValue=2, group=Center Orb | Gradients, parameterName=Bottom Up Glow Contrast, sortPriority=60}) {materialFunction=&MF_Fresnel_FX}
  n44 = Clamp(
   None=@n38.Bottom Up Alpha
   Max=ScalarParameter() {defaultValue=0.3, group=Center Orb | Gradients, parameterName=Bottom Up Glow Max Brightness, sortPriority=60}
  ) {maxDefault=0.5}
  n45 = Saturate(None=Add(A=Multiply(A=@n41.RadialGradientExponential, B=@n42) {constB=5}, B=Multiply(A=@n43.Fresnel, B=@n44)))
  n46 = Add(A=@n36, B=@n45)
  n47 = ScalarParameter() {defaultValue=1.2, group=Center Orb | Gradients, parameterName=Gradient Masks Intensity, sortPriority=50}
  NamedRerouteDeclaration(None=Saturate(None=Multiply(A=@n46, B=@n47))) {name=Gradient Masks, nodeColor=<rgba 1 0.2578 0 1>}
  n48 = NamedRerouteUsage()
  n49 = ComponentMask(None=NamedRerouteUsage()) {b=true}
  n50 = ScreenPosition()
  n51 = Multiply(
   A=MaterialFunctionCall(World Position=NamedRerouteUsage()) {materialFunction=&TransformToClipSpace}.Clip Space XY
   B=Constant2Vector() {g=-1, r=1}
  )
  n52 = NamedRerouteUsage()
  n53 = ScalarParameter() {defaultValue=10, group=Center Orb | Constellation, parameterName=Constellation Map Scale, sortPriority=41}
  n54 = MaterialFunctionCall(
   UVs=Subtract(A=Add(A=@n49, B=@n50.ViewportUV), B=Multiply(A=@n51) {constB=0.5})
   Texture Scale=Multiply(A=Multiply(A=@n52, B=@n53) {constB=1.1}, B=NamedRerouteUsage())
  ) {materialFunction=&ScaleUVsByCenter}
  n55 = NamedRerouteUsage()
  n56 = Multiply(A=MaterialFunctionCall(X=ComponentMask(None=@n55) {r=true}) {materialFunction=&MakeFloat2}.Result) {constB=-0.33}
  n57 = Multiply(A=NamedRerouteUsage()) {constB=0.175}
  n58 = TextureObjectParameter() {
   group=Center Orb | Constellation
   parameterName=Constellation Map Background
   sortPriority=47
   texture=/Game/Environments/Helios/Props/CentralPicnic/Textures/T_CentralPicnic_ChestOrbConstellations
  }
  n59 = TextureSample(
   UVs=Add(
    A=Add(A=@n54.UVs, B=@n56)
    B=MaterialFunctionCall(Y=@n57) {materialFunction=&MakeFloat2}.Result
   )
   Tex=@n58
  )
  n60 = Multiply(
   A=@n59.R
   B=ScalarParameter() {defaultValue=25, group=Center Orb | Constellation, parameterName=Constellation Stars Intensity, sortPriority=43}
  )
  n61 = Multiply(
   A=@n59.G
   B=ScalarParameter() {defaultValue=35, group=Center Orb | Constellation, parameterName=Constellation Connection Base Intensity, sortPriority=43}
  )
  n62 = Multiply(
   A=Multiply(A=OneMinus(None=@n48), B=Max(A=@n60, B=@n61))
   B=MaterialFunctionCall(
    Brightness=ScalarParameter() {defaultValue=1.5, group=Center Orb | Constellation, parameterName=Constellation Mask Desnity}
    Exponent=ScalarParameter() {defaultValue=8, group=Center Orb | Constellation, parameterName=Constellation Edge Mask}
   ) {materialFunction=&MF_Fresnel_FX}.Fresnel
  )
  n63 = NamedRerouteUsage()
  n64 = Constant() {r=0.5}
  n65 = Multiply(A=Clamp(None=Distance(A=@n63, B=@n64)) {maxDefault=0.5}) {constB=2}
  NamedRerouteDeclaration(None=LinearInterpolate(A=Saturate(None=@n62), Alpha=Power(Base=@n65) {constExponent=10}) {constB=0}) {name=InnerConstellationMask, nodeColor=<rgba 0 1 1 1>}
  n66 = MaterialFunctionCall(WarpMask=ComponentMask(None=NamedRerouteUsage()) {r=true}) {materialFunction=&MF_CP_ChestGems_Utilities}
  NamedRerouteDeclaration(None=@n66.Center Orb Distance) {name=Orb Center UV Distance, nodeColor=<rgba 1 0 0.0469 1>}
  NamedRerouteDeclaration(None=@n66.FOV - Scale Multiplier) {name=FOV Scalar, nodeColor=<rgba 0.2656 0 1 1>}
  NamedRerouteDeclaration(None=@n66.Distance Mask) {name=Distance Mask, nodeColor=<rgba 0.5703 0 1 1>}
  NamedRerouteDeclaration(None=@n66.Glass Color) {name=Glass Color, nodeColor=<rgba 1 0 0.0469 1>}
  NamedRerouteDeclaration(None=@n66.XY Rotation Alpha) {name=XY Rotation Alpha, nodeColor=<rgba 1 0.2578 0 1>}
  NamedRerouteDeclaration(None=@n66.Gem Roughness) {name=Roughness, nodeColor=<rgba 1 0 0.3047 1>}
  n67 = VectorParameter() {group=_Colors, parameterName=Center Orb Front Constellation Color, sortPriority=19, defaultValue=<rgba 1 0.2551 0 1>}
  n68 = ScalarParameter() {defaultValue=10, group=Center Orb | Front Constellations, parameterName=Front Constellation Aux Stars Tiling, sortPriority=26}
  n69 = Multiply(A=Multiply(A=NamedRerouteUsage(), B=@n68) {constB=0.1}, B=NamedRerouteUsage())
  n70 = NamedRerouteUsage()
  n71 = MaterialFunctionCall(World Position=NamedRerouteUsage()) {materialFunction=&TransformToClipSpace}
  n72 = Constant2Vector() {g=-1, r=1}
  n73 = Subtract(A=Add(A=ComponentMask(None=@n70) {r=true}, B=ScreenPosition().ViewportUV), B=Multiply(A=Multiply(A=@n71.Clip Space XY, B=@n72)) {constB=0.5})
  n74 = Multiply(A=NamedRerouteUsage()) {constB=0.2}
  n75 = TextureObjectParameter() {
   group=Center Orb | Constellation
   parameterName=Constellation Map Background
   sortPriority=47
   texture=&T_CentralPicnic_ChestOrbConstellations
  }
  n76 = TextureSample(
   UVs=Add(
    A=MaterialFunctionCall(UVs=@n73, Texture Scale=@n69) {materialFunction=&ScaleUVsByCenter}.UVs
    B=MaterialFunctionCall(Y=@n74) {materialFunction=&MakeFloat2}.Result
   )
   Tex=@n75
  )
  n77 = ScalarParameter() {defaultValue=10, group=Center Orb | Front Constellations, parameterName=Front Constellation Aux Stars Multi, sortPriority=26}
  n78 = Time()
  n79 = VectorParameter() {
   group=Center Orb | Front Constellations
   parameterName=Front Constellation Pulse Settings
   sortPriority=22
   channelNames {a=Contrast, b=Tiling, g=Speed Y, r=Speed X}
   defaultValue=<rgba 0.05 -0.05 2 6>
  }
  n80 = MaterialFunctionCall(X=@n79.Speed X, Y=@n79.Speed X) {materialFunction=&MakeFloat2}
  n81 = TextureObjectParameter() {
   group=_General
   parameterName=General Cloud Noise
   samplerType=SAMPLERTYPE_LinearGrayscale
   texture=&T_SimplePerlin_Greyscale
  }
  n82 = TextureSample(UVs=Multiply(A=Panner(Time=@n78, Speed=@n80.Result) {speedX=0.01, speedY=0.05}, B=@n79.Tiling), Tex=@n81) {samplerType=SAMPLERTYPE_LinearGrayscale, texture=&T_SimplePerlin_Greyscale}
  n83 = Multiply(A=Multiply(A=ComponentMask(None=@n76.R) {r=true}, B=@n77) {constB=5}, B=Power(Base=@n82.R, Exp=@n79.Contrast) {constExponent=4})
  n84 = StaticSwitchParameter(True=ScalarParameter() {group=Center Orb | Front Constellations, parameterName=FlipbookAnimPhase, sortPriority=2}, False=NamedRerouteUsage()) {group=Center Orb | Front Constellations, parameterName=StaticFlipbookAnimPhase, sortPriority=1}
  n85 = ScalarParameter() {defaultValue=2, group=Center Orb | Front Constellations, parameterName=Front Constellations Num Flipbook Rows}
  n86 = ScalarParameter() {defaultValue=4, group=Center Orb | Front Constellations, parameterName=Front Constellations Num Flipbook Columns}
  n87 = TextureObjectParameter() {
   group=Center Orb | Constellation
   parameterName=Constellation Map Front
   sortPriority=47
   texture=/Game/Environments/Helios/Props/CentralPicnic/Textures/T_CentralPicnic_ChestOrbConstellations_FrontFlipbook
  }
  n88 = StaticSwitchParameter(
   True=ScalarParameter() {group=Center Orb | Front Constellations, parameterName=StaticFlipbookAnimTime, sortPriority=2}
   False=Multiply(
    A=Time()
    B=ScalarParameter() {defaultValue=0.01, group=Center Orb | Front Constellations, parameterName=Front Constellation Speed, sortPriority=26}
   ) {constB=0.1}
  ) {group=Center Orb | Front Constellations, parameterName=StaticFlipbookAnimPhase, sortPriority=1}
  n89 = ScalarParameter() {defaultValue=2, group=Center Orb | Front Constellations, parameterName=Front Constellations Num Flipbook Rows}
  n90 = ScalarParameter() {defaultValue=4, group=Center Orb | Front Constellations, parameterName=Front Constellations Num Flipbook Columns}
  n91 = Add(A=Multiply(A=NamedRerouteDeclaration(None=@n88) {name=FlipbookSpeed, nodeColor=<rgba 0 1 0.4141 1>}, B=Multiply(A=@n89, B=@n90)) {constB=4}) {constB=0.75}
  n92 = Sine(None=@n91)
  n93 = LinearInterpolate(
   A=ScalarParameter() {defaultValue=0.125, group=Center Orb | Front Constellations, parameterName=Front Constellation Warp Min, sortPriority=26}
   B=ScalarParameter() {defaultValue=0.005, group=Center Orb | Front Constellations, parameterName=Front Constellation Warp Max, sortPriority=26}
   Alpha=Power(Base=ConstantBiasScale(None=@n92)) {constExponent=0.25}
  )
  n94 = NamedRerouteUsage()
  n95 = Multiply(A=@n93, B=Multiply(A=Subtract(A=@n94) {constB=0.5}))
  n96 = ScalarParameter() {defaultValue=0.6, group=Center Orb | Front Constellations, parameterName=Front Constellation Stars Warp Mult, sortPriority=26}
  n97 = MaterialFunctionCall(Float2=ViewSize()) {materialFunction=&BreakOutFloat2Components}
  n98 = Divide(A=@n97.G, B=@n97.G)
  n99 = Multiply(
   A=NamedRerouteUsage()
   B=ScalarParameter() {defaultValue=3, group=Center Orb | Front Constellations, parameterName=Front Constellation Tiling, sortPriority=26}
  ) {constB=1.1}
  n100 = NamedRerouteUsage()
  n101 = Multiply(
   A=MaterialFunctionCall(X=Constant() {r=1}, Y=@n98) {materialFunction=&MakeFloat2}.Result
   B=MaterialFunctionCall(UVs=@n73, Texture Scale=Multiply(A=@n99, B=@n100)) {materialFunction=&ScaleUVsByCenter}.UVs
  )
  n102 = MaterialFunctionCall(
   X=ScalarParameter() {group=Center Orb | Front Constellations, parameterName=Front Constellation Pos Offset X}
   Y=ScalarParameter() {defaultValue=-0.1, group=Center Orb | Front Constellations, parameterName=Front Constellation Pos Offset Y}
  ) {materialFunction=&MakeFloat2}
  n103 = Divide(A=OneMinus(None=@n98))
  n104 = ScalarParameter() {
   defaultValue=0.1
   desc="Use when using more intense warping values, to keep it centered duting warps in/out"
   group=Center Orb | Front Constellations
   parameterName=Front Constellation Warp X Offset
   sortPriority=26
  }
  n105 = NamedRerouteUsage()
  n106 = Divide(B=Multiply(
    A=ScalarParameter() {defaultValue=2, group=Center Orb | Front Constellations, parameterName=Front Constellations Num Flipbook Rows}
    B=ScalarParameter() {defaultValue=4, group=Center Orb | Front Constellations, parameterName=Front Constellations Num Flipbook Columns}
   ))
  n107 = Add(
   A=Add(
    A=MaterialFunctionCall(Y=@n103) {materialFunction=&MakeFloat2}.Result
    B=Add(A=@n101, B=@n102.Result)
   )
   B=MaterialFunctionCall(
    X=Multiply(A=@n93, B=@n104) {constB=-0.1}
    Y=MaterialFunctionCall!(Time=@n105, Divisor, Mod=@n106) {materialFunction=/Game/Effects/Fort_Effects/Materials/_Functions/MF_ModuloTime}.Unclamped
   ) {materialFunction=&MakeFloat2}.Result
  )
  n108 = MaterialFunctionCall(
   "Animation  Phase (0-1)"=@n84
   Number of Rows=@n85
   Number of Columns=@n86
   Texture=@n87
   UVs=Add(A=@n107, B=Multiply(A=@n95, B=@n96) {constB=0.5})
  ) {desc=Less warping is used on the stars, materialFunction=&FlipBook}
  n109 = ComponentMask(None=@n108.Result) {b=true}
  n110 = Add(
   A=ComponentMask(None=@n108.Result) {r=true}
   B=Multiply(
    A=@n109
    B=ScalarParameter() {defaultValue=0.1, group=Center Orb | Front Constellations, parameterName=Front Constellation Star Bloom, sortPriority=22}
   )
  )
  n111 = ConstantBiasScale(None=Sine(None=@n91))
  n112 = VectorParameter() {
   parameterName=FadeControls
   channelNames {a=Connecitons Constast, b=Connections Multi, g=Stars Contrast, r=Stars Multi}
   defaultValue=<rgba 8 1 2 2>
  }
  n113 = Power(Base=Saturate(None=Multiply(A=@n111, B=@n112.Stars Multi) {constB=8}), Exp=@n112.Stars Contrast) {constExponent=1}
  n114 = Power(Base=Saturate(None=Multiply(A=@n111, B=@n112.Connections Multi) {constB=2}), Exp=@n112.Connecitons Constast)
  n115 = ComponentMask(None=MaterialFunctionCall("Animation  Phase (0-1)"=@n84, Number of Rows=@n85, Number of Columns=@n86, Texture=@n87, UVs=Add(A=@n107, B=@n95)) {materialFunction=&FlipBook}.Result) {g=true}
  n116 = MaterialFunctionCall(Float3=MaterialFunctionCall(X=Multiply(A=@n113, B=@n110), Y=Multiply(A=@n114, B=@n115), Z=Multiply(A=@n113, B=@n109)) {materialFunction=&MakeFloat3}.Result) {materialFunction=&BreakOutFloat3Components}
  n117 = VectorParameter() {
   group=Center Orb | Front Constellations
   parameterName=Front Constellation Intensity Mult
   sortPriority=21
   channelNames {a=Connections Pulse, b=Connections, g=Star Pulse, r=Stars}
   defaultValue=<rgba 1 1.5 0.75 2>
  }
  n118 = Power(Base=Multiply(A=@n114, B=@n82.R), Exp=@n79.Contrast) {constExponent=4}
  n119 = LinearInterpolate(B=@n116.R, Alpha=@n118)
  n120 = Add(
   A=Add(A=Multiply(A=@n116.R, B=@n117.Stars), B=Multiply(A=@n119, B=@n117.Star Pulse) {constB=10})
   B=Multiply(A=LinearInterpolate(B=@n118, Alpha=@n116.G), B=@n117.Connections Pulse) {constB=10}
  )
  n121 = VectorParameter() {group=_Colors, parameterName=Center Orb Constellation Pulse Color, sortPriority=19, defaultValue=<rgba 0.6392 0.8471 1 1>}
  n122 = Multiply(A=Add(A=@n83, B=@n120), B=LinearInterpolate(A=@n67.RGB, B=@n121.RGB, Alpha=@n118))
  n123 = NamedRerouteUsage()
  NamedRerouteDeclaration(None=Multiply(A=LinearInterpolate(A=@n67.RGB, B=@n122, Alpha=@n120), B=LinearInterpolate(Alpha=@n123) {constA=1, constB=2})) {name=FrontConstellationColors, nodeColor=<rgba 1 0.1694 0 1>}
  n124 = Constant()
  n125 = MaterialFunctionCall(WarpMask=NamedRerouteDeclaration(None=FunctionInput(Preview=@n124) {bUsePreviewValueAsDefault=true, inputName=WarpMask, sortPriority=4}) {name=WarpMaskInput, nodeColor=<rgba 0 0.1797 1 1>}) {materialFunction=&MF_CP_ChestGems_Utilities}
  n126 = Constant() {r=0.4}
  n127 = Constant() {r=1}
  n128 = MaterialFunctionCall(Float2=ViewSize()) {materialFunction=&BreakOutFloat2Components}
  n129 = Divide(A=@n128.G, B=@n128.G)
  n130 = NamedRerouteUsage()
  n131 = MaterialFunctionCall(World Position=NamedRerouteUsage()) {materialFunction=&TransformToClipSpace}
  n132 = Constant2Vector() {g=-1, r=1}
  n133 = Subtract(
   A=Add(A=ComponentMask(None=@n130) {r=true}, B=ScreenPosition().ViewportUV)
   B=Multiply(A=Multiply(A=@n131.Clip Space XY, B=@n132)) {constB=0.5}
  )
  n134 = ScalarParameter() {defaultValue=2, group=Center Orb | Wisp, parameterName=Wisp Size X, sortPriority=14}
  n135 = ScalarParameter() {defaultValue=3, group=Center Orb | Wisp, parameterName=Wisp Size Y, sortPriority=14}
  n136 = Multiply(A=Multiply(A=NamedRerouteUsage(), B=AppendVector(A=@n134, B=@n135)) {constB=1.1}, B=NamedRerouteUsage())
  n137 = OneMinus(None=@n129)
  n138 = Add(
   A=Multiply(
    A=MaterialFunctionCall(X=@n127, Y=@n129) {materialFunction=&MakeFloat2}.Result
    B=MaterialFunctionCall(UVs=@n133, Texture Scale=@n136) {materialFunction=&ScaleUVsByCenter}.UVs
   )
   B=MaterialFunctionCall(Y=Divide(A=@n137)) {materialFunction=&MakeFloat2}.Result
  )
  n139 = Multiply(A=NamedRerouteUsage()) {constB=0.2}
  n140 = Add(
   A=Multiply(A=@n138, B=MaterialFunctionCall(X=@n126, Y=@n126) {materialFunction=&MakeFloat2}.Result)
   B=@n139
  )
  n141 = TextureSample(UVs=Add(A=Panner(Coordinate=@n138) {speedY=-0.05}, B=@n139)) {samplerType=SAMPLERTYPE_Normal, texture=/Game/Accessories/FORT_Backpacks/Backpack_M_MED_EctoCat/FX/CloudNoise_T_N}
  n142 = AppendVector(A=@n141.R, B=@n141.R)
  n143 = ScalarParameter() {defaultValue=0.01, group=Center Orb | Wisp, parameterName=DistortionIntensity}
  n144 = Constant()
  n145 = FeatureLevelSwitch(Default=QualitySwitch(Default=Multiply(A=@n142, B=@n143), Low=@n144) {desc=Remove subtle distortion on low}, ES3_1=@n144)
  n146 = MaterialFunctionCall(X=Constant() {r=-0.05}, Y=Constant() {r=0.1}) {materialFunction=&MakeFloat2}
  n147 = ScalarParameter() {defaultValue=0.75, group=Center Orb | Wisp, parameterName=Wisp Speed}
  n148 = TextureObjectParameter() {
   group=Center Orb | Wisp
   parameterName=Smoke Wisp Texture
   samplerType=SAMPLERTYPE_LinearGrayscale
   sortPriority=5
   texture=/Game/Global/Textures/Default/Flowmaps/T_StormSwirlMask
  }
  n149 = TextureSample(UVs=Panner(Coordinate=Add(A=@n140, B=@n145), Speed=Multiply(A=@n146.Result, B=@n147)) {speedX=-0.05, speedY=0.1}, Tex=@n148) {samplerSource=SSM_Wrap_WorldGroupSettings, samplerType=SAMPLERTYPE_Masks, texture=/Game/Weapons/WeaponSkins/Wraps/Textures/Mash/Mash_Noise}
  n150 = Constant() {r=0.45}
  n151 = Constant() {r=0.5}
  n152 = Add(
   A=Multiply(A=@n138, B=MaterialFunctionCall(X=@n150, Y=@n151) {materialFunction=&MakeFloat2}.Result)
   B=@n139
  )
  n153 = MaterialFunctionCall(X=Constant() {r=0.045}, Y=Constant() {r=0.1}) {materialFunction=&MakeFloat2}
  n154 = TextureSample(UVs=Panner(Coordinate=Add(A=@n152, B=@n145), Speed=Multiply(A=@n147, B=@n153.Result)) {speedX=0.045, speedY=0.1}, Tex=@n148) {samplerSource=SSM_Wrap_WorldGroupSettings, samplerType=SAMPLERTYPE_Masks, texture=&Mash_Noise}
  n155 = Multiply(A=@n149.R, B=@n154.R)
  n156 = Multiply(
   A=NamedRerouteDeclaration(None=@n155) {name=Wisp Texture, nodeColor=<rgba 1 0.0469 0 1>}
   B=ScalarParameter() {defaultValue=1, group=Center Orb | Wisp, parameterName=Primary Wisp Warp Amount, sortPriority=24}
  ) {constB=0.2}
  n157 = MaterialFunctionCall(
   UVs=Add(A=@n125.Screen Scaled UVs, B=Multiply(A=@n156, B=@n125.Center Orb Distance)) {constB=0}
   CenterPosition=@n125.Orb Center UV Space
   Radius=Multiply(
    A=ScalarParameter() {defaultValue=2, group=Center Orb | Wisp, parameterName=Primary Wisp Radius, sortPriority=25}
    B=Multiply(A=@n125.Center Orb Distance, B=@n125.FOV - Scale Multiplier)
   )
   Density=ScalarParameter() {defaultValue=4, group=Center Orb | Wisp, parameterName=Primary Wisp Density, sortPriority=25}
  ) {materialFunction=&RadialGradientExponential}
  n158 = Multiply(A=@n157.RadialGradientExponential, B=@n155)
  n159 = ScalarParameter() {defaultValue=0.75, group=Center Orb | Wisp, parameterName=Primary Wisp Detail Blend, sliderMax=1, sortPriority=26}
  n160 = Multiply(
   A=Power(
    Base=LinearInterpolate(A=@n157.RadialGradientExponential, B=@n158, Alpha=@n159)
    Exp=ScalarParameter() {defaultValue=1, group=Center Orb | Wisp, parameterName=Primary Wisp Constrast, sortPriority=26}
   )
   B=ScalarParameter() {defaultValue=1.5, group=Center Orb | Wisp, parameterName=Primary Wisp Mult, sortPriority=27}
  )
  n161 = Power(
   Base=Saturate(None=@n160)
   Exp=ScalarParameter() {defaultValue=10, group=Center Orb | Wisp, parameterName=Wisp Background Mask Contrast, sortPriority=50}
  )
  n162 = ScalarParameter() {defaultValue=0.1, group=Center Orb | Wisp, parameterName=Wisp Background Mask Multiplier, sortPriority=50}
  NamedRerouteDeclaration(None=Saturate(None=Multiply(A=@n161, B=@n162))) {name=Wisp Background Mask, nodeColor=<rgba 0 1 0.1328 1>}
  n163 = NamedRerouteUsage()
  n164 = ScalarParameter() {defaultValue=2, group=Center Orb | Front Constellations, parameterName=Front Constellation Wisp Mask Contrast, sortPriority=50}
  n165 = ScalarParameter() {defaultValue=1.75, group=Center Orb | Front Constellations, parameterName=Front Constellation Wisp Mask Mult, sortPriority=50}
  n166 = Multiply(A=Power(Base=Saturate(None=@n163), Exp=@n164), B=@n165)
  n167 = Multiply(A=OneMinus(None=Saturate(None=@n166)), B=Max(A=@n120, B=Multiply(A=@n116.G, B=@n117.Connections)))
  NamedRerouteDeclaration(None=Add(A=@n83, B=NamedRerouteDeclaration(None=@n167) {name=FrontConstellationMask_OnlyMainCons, nodeColor=<rgba 0.1719 1 0 1>})) {name=FrontConstellationMaskAndStars, nodeColor=<rgba 0 0.4141 1 1>}
  NamedRerouteDeclaration(None=@n66.Glass Thickness Mask) {name=Glass Thickness Mask, nodeColor=<rgba 0.2656 0 1 1>}
  NamedRerouteDeclaration(None=@n66.XY Rotation Offset Amount) {name=XY Rotation Offset, nodeColor=<rgba 1 0.2578 0 1>}
  NamedRerouteDeclaration(None=@n66.Orb Center UV Space) {name=Orb Center UV Space, nodeColor=<rgba 1 0 0.7969 1>}
  n168 = MaterialFunctionCall() {materialFunction=&MF_Fresnel_FX}
  n169 = OneMinus(None=@n149.R)
  n170 = OneMinus(None=@n154.R)
  n171 = Multiply(A=Add(A=Multiply(A=@n169) {constB=-1}, B=@n170)) {constB=0.2}
  n172 = MaterialFunctionCall(
   UVs=Add(A=@n125.Screen Scaled UVs, B=Multiply(A=@n171, B=@n125.Center Orb Distance)) {constB=0}
   CenterPosition=@n125.Orb Center UV Space
   Radius=Multiply(
    A=ScalarParameter() {defaultValue=1.8, group=Center Orb | Wisp, parameterName=Secondary Wisp Radius, sortPriority=40}
    B=Multiply(A=@n125.Center Orb Distance, B=@n125.FOV - Scale Multiplier)
   )
   Density=ScalarParameter() {defaultValue=2, group=Center Orb | Wisp, parameterName=Secondary Wisp Density, sortPriority=41}
  ) {materialFunction=&RadialGradientExponential}
  n173 = Multiply(A=@n169, B=@n170)
  n174 = ScalarParameter() {defaultValue=1.5, group=Center Orb | Wisp, parameterName=Secondary Wisp Power, sortPriority=40}
  n175 = NamedRerouteDeclaration(None=Multiply(A=@n172.RadialGradientExponential, B=Power(Base=@n173, Exp=@n174) {constExponent=1.5})) {name=Inner Wisp Mask B, nodeColor=<rgba 1 0.6797 0 1>}
  n176 = ScalarParameter() {defaultValue=0.26, group=Center Orb | Wisp, parameterName=Secondary Wisp Detail Blend, sliderMax=1, sortPriority=42}
  n177 = Add(
   A=NamedRerouteDeclaration(None=@n160) {name=Inner Wisp Mask A, nodeColor=<rgba 0 0.2031 1 1>}
   B=Multiply(
    A=LinearInterpolate(A=@n172.RadialGradientExponential, B=@n175, Alpha=@n176)
    B=ScalarParameter() {defaultValue=1, group=Center Orb | Wisp, parameterName=Secondary Wisp Mutl, sortPriority=41}
   )
  )
  NamedRerouteDeclaration(None=Multiply(A=Power(Base=@n168.Fresnel), B=Saturate(None=@n177))) {name=Inner Wisp Mask Merged, nodeColor=<rgba 0.4766 1 0 1>}
  n178 = VectorParameter() {group=_Colors, parameterName=Center Orb Gradient B Color, sortPriority=21, defaultValue=<rgba 1 0.2114 0.047 1>}
  n179 = VectorParameter() {group=_Colors, parameterName=Center Orb Gradient A Color, sortPriority=21, defaultValue=<rgba 1 0.6564 0.2344 1>}
  NamedRerouteDeclaration(None=LinearInterpolate(A=LinearInterpolate(B=@n178.RGB, Alpha=@n36), B=@n179.RGB, Alpha=@n45)) {name=Gradient Colors, nodeColor=<rgba 1 0 0.7031 1>}
  NamedRerouteDeclaration(None=@n66.Screen Scaled UVs) {name=Screen Scaled UVs, nodeColor=<rgba 0.9219 0 1 1>}
  NamedRerouteDeclaration(None=@n66.Orb Center World Pos) {name=Orb Center World Pos, nodeColor=<rgba 1 0.3984 0 1>}
  n180 = Constant()
  Reroute(None=@n180)
  Reroute(None=@n180)
 ]}
